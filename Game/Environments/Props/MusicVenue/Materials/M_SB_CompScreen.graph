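# Unlit, Nanite-enabled material for a computer screen prop in the music venue, showing an animated color pattern with scanlines.
# Two rotating UV sets are summed to sample a borrowed backpack radial gradient, whose color gets a time-oscillating hue shift.
# Screen-space sine scanlines, bent by Fresnel and view angle except on low and medium quality, fade with camera distance.
# The scanlines both offset one UV set and add a faint brightness.
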Material /Game/Environments/Props/MusicVenue/Materials/M_SB_CompScreen
graph {
 properties {bUsedWithNanite=true, pixelDepthOffsetMode=PDOM_Legacy, shadingModel=MSM_Unlit, usedShadingModels=Unlit}
 nodes [
  n1 = NamedRerouteUsage()
  n2 = Constant() {r=0.3}
  n3 = LinearInterpolate(Alpha=Clamp(None=MaterialFunctionCall(Speed=@n2) {materialFunction=&DebugTimeSine}.Result)) {constB=0.35}
  n4 = Add(A=NamedRerouteUsage(), B=TextureCoordinate())
  n5 = TextureCoordinate()
  n6 = TextureSample(UVs=Add(A=Rotator(Coordinate=@n4) {speed=1}, B=Rotator(Coordinate=@n5) {speed=0.1})) {texture=/Game/Accessories/FORT_Backpacks/M_MED_Color_Block/Textures/T_M_MED_Color_Block_Backpack_Gradient_Radial}
  n7 = AppendVector(A=AppendVector(A=@n6.R, B=@n6.R), B=Constant() {r=1})
  Clamp(None=Add(
    A=Clamp(None=@n1) {maxDefault=0.1}
    B=MaterialFunctionCall(Hue Shift Percentage=@n3, Texture=@n7) {materialFunction=&HueShift}.Result
   ))
  n8 = Fresnel() {baseReflectFraction=0, exponent=1}
  n9 = DotProduct(A=ComponentMask(None=VertexNormalWS()) {b=true}, B=ComponentMask(None=CameraVectorWS()) {b=true})
  n10 = Constant()
  n11 = Add(
   A=QualitySwitch(Default=Multiply(A=@n8, B=@n9) {constB=0.2}, Low=@n10, Medium=@n10)
   B=ComponentMask(None=ScreenPosition().ViewportUV) {g=true}
  )
  n12 = ScalarParameter() {defaultValue=75, parameterName=ScanlineTile}
  n13 = ConstantBiasScale(None=Sine(None=Multiply(A=@n11, B=@n12)))
  n14 = ScalarParameter() {defaultValue=2000, parameterName=FadeDistance}
  NamedRerouteDeclaration(None=LinearInterpolate(
    A=Power(Base=@n13)
    Alpha=MaterialFunctionCall(Distance=@n14) {materialFunction=&Near_Camera_Fade}.Result
   ) {constA=0.4, constB=0.3}) {name=Scanline, nodeColor=<rgba 0 0.2969 1 1>}
 ]
}
